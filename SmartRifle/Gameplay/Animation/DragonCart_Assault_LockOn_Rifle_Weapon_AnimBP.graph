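# Animation Blueprint for the DragonCart assault lock-on rifle that animates its optic between non-targeted and targeted poses.
# A frozen non-targeted pose feeds DefaultSlot, and Targeting SM is layered over it from the optic_body bone.
# bIsTargeting drives the state machine through transition clips that auto-advance to held poses; retargeting can interrupt the return clip.
# The event graph is empty, and the declared ABP_IsTargeting variable is unused.
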AnimBlueprint /SmartRifle/Gameplay/Animation/DragonCart_Assault_LockOn_Rifle_Weapon_AnimBP
target_skeleton=/SmartRifle/Gameplay/Meshes/Assault_LockOn_Rifle_Skeleton
graphs [
 AnimGraph {nodes [
   n1 = Sequence Player 'DragonCart_Assault_LockOn_Rifle_NonTargetedPose'() {bLoopAnimation=false, playRate=0, sequence=&/DragonCart_Assault_LockOn_Rifle_NonTargetedPose}
   AnimGraphNode_StateMachine_0 = Targeting SM() {is=title AnimGraphNode_StateMachine}
   Output Pose(Result=Layered blend per bone(BasePose=Slot 'DefaultSlot'(Source=@n1), BlendPoses_0=@AnimGraphNode_StateMachine_0) {layerSetup [{branchFilters [{boneName=optic_body}]}]})
   Get bIsTargeting()
  ]}
 Targeting SM {
  entry=Non Targeted
  path=AnimGraph.AnimGraphNode_StateMachine_0.Targeting SM
  ids=id
  nodes [
   AnimStateNode_0 = Non Targeted() {is=title AnimStateNode}
   AnimStateNode_1 = Targeted() {is=title AnimStateNode}
   AnimStateNode_2 = TransitionToTargeted() {is=title AnimStateNode}
   AnimStateNode_3 = TransitionToNonTargeted() {is=title AnimStateNode}
   AnimStateTransitionNode_0 = AnimStateTransitionNode() {bAllowInertializationForSelfTransitions=true}
   AnimStateTransitionNode_2 = AnimStateTransitionNode() {bAllowInertializationForSelfTransitions=true, crossfadeDuration=0}
   AnimStateTransitionNode_3 = AnimStateTransitionNode() {bAllowInertializationForSelfTransitions=true, bAutomaticRuleBasedOnSequencePlayerInState=true, crossfadeDuration=0}
   AnimStateTransitionNode_4 = AnimStateTransitionNode() {bAllowInertializationForSelfTransitions=true, crossfadeDuration=0}
   AnimStateTransitionNode_5 = AnimStateTransitionNode() {bAllowInertializationForSelfTransitions=true, bAutomaticRuleBasedOnSequencePlayerInState=true, crossfadeDuration=0}
  ]
  links [
   {
    dst=TransitionToTargeted
    rule=AnimGraph.AnimGraphNode_StateMachine_0.Targeting SM.AnimStateTransitionNode_2.Transition
    src=Non Targeted
    via=AnimStateTransitionNode_2
   }
   {
    dst=TransitionToNonTargeted
    rule=AnimGraph.AnimGraphNode_StateMachine_0.Targeting SM.AnimStateTransitionNode_4.Transition
    src=Targeted
    via=AnimStateTransitionNode_4
   }
   {
    dst=Non Targeted
    rule=AnimGraph.AnimGraphNode_StateMachine_0.Targeting SM.AnimStateTransitionNode_5.Transition
    src=TransitionToNonTargeted
    via=AnimStateTransitionNode_5
   }
   {
    dst=TransitionToTargeted
    rule=AnimGraph.AnimGraphNode_StateMachine_0.Targeting SM.AnimStateTransitionNode_2.Transition
    src=TransitionToNonTargeted
    via=AnimStateTransitionNode_0
   }
   {
    dst=Targeted
    rule=AnimGraph.AnimGraphNode_StateMachine_0.Targeting SM.AnimStateTransitionNode_3.Transition
    src=TransitionToTargeted
    via=AnimStateTransitionNode_3
   }
  ]
 }
 Non Targeted {
  path=AnimGraph.AnimGraphNode_StateMachine_0.Targeting SM.AnimStateNode_0.Non Targeted
  nodes [Output Animation Pose(Result=Sequence Player 'DragonCart_Assault_LockOn_Rifle_NonTargetedPose'() {sequence=&DragonCart_Assault_LockOn_Rifle_NonTargetedPose})]
 }
 Targeted {
  path=AnimGraph.AnimGraphNode_StateMachine_0.Targeting SM.AnimStateNode_1.Targeted
  nodes [Output Animation Pose(Result=Sequence Player 'DragonCart_Assault_LockOn_Rifle_TargetedPose'() {playRate=0, sequence=&/DragonCart_Assault_LockOn_Rifle_TargetedPose})]
 }
 TransitionToTargeted {
  path=AnimGraph.AnimGraphNode_StateMachine_0.Targeting SM.AnimStateNode_2.TransitionToTargeted
  nodes [
   n1 = Sequence Player 'DragonCart_Assault_LockOn_Rifle_TransitionToTargeted'() {bLoopAnimation=false, playRate=1.5, sequence=&/DragonCart_Assault_LockOn_Rifle_TransitionToTargeted}
   Output Animation Pose(Result=@n1)
  ]
 }
 TransitionToNonTargeted {
  path=AnimGraph.AnimGraphNode_StateMachine_0.Targeting SM.AnimStateNode_3.TransitionToNonTargeted
  nodes [Output Animation Pose(Result=Sequence Player 'DragonCart_Assault_LockOn_Rifle_TransitionToNonTargeted'() {bLoopAnimation=false, sequence=&/DragonCart_Assault_LockOn_Rifle_TransitionToNonTargeted})]
 }
 Transition {
  path=AnimGraph.AnimGraphNode_StateMachine_0.Targeting SM.AnimStateTransitionNode_2.Transition
  nodes [Result(bCanEnterTransition=Get bIsTargeting())]
 }
 Transition {path=AnimGraph.AnimGraphNode_StateMachine_0.Targeting SM.AnimStateTransitionNode_3.Transition, nodes [Result()]}
 Transition {
  path=AnimGraph.AnimGraphNode_StateMachine_0.Targeting SM.AnimStateTransitionNode_4.Transition
  nodes [Result(bCanEnterTransition=NOT Boolean(A=Get bIsTargeting()))]
 }
 Transition {path=AnimGraph.AnimGraphNode_StateMachine_0.Targeting SM.AnimStateTransitionNode_5.Transition, nodes [Result()]}
 EventGraph {links [], nodes []}
]
variables [ABP_IsTargeting {type=boolean}]
