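; Pawn component for the Smart Rifle auto-aim weapon that builds stacking lock-ons on a target while aiming and fires bursts at it.
; Client tick finds and keeps a reticle target, charging lock-on stacks capped by magazine ammo, with audio, material and UI feedback.
; Burst completion spends one stack and timers refire while the trigger or locked-on auto-fire persists.
; A replicated lock target drives the victim's warning sound and camera effect.
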(event Custom|OnReceiveFirstLockOn
  (Effects|SetBeamNiagaraSystem (Variables|Default|GetCachedWeapon) "/SmartRifle/Gameplay/VFX/NS_AutoAimRiffle_Tracer.NS_AutoAimRiffle_Tracer"))

(event EventBeginPlay
  ;[1 1 1] Begin Play
  (CallFunction|InitializeAutoAimComponent)
  ; On Tact Sprint, reset our cartridges as interrupting mid-fire causes issues.
  (if (Class|AthenaFunctionLibrary|RowtoBool (Variables|Hotfix|GetHF_ResetCartridgesPostTactSprint))
    (Ability|Tasks|WaitGameplayTagRemoveFromActor (Components|GetOwner) "(TagName=\"Gameplay.Action.Player.TacticalSprint\")"
      (:Removed
        (Variables|Default|GetCachedWeapon
          (:then
            (Variables|Firing|SetCartridgesLeftToFire 0))
          (:else))))))

(event Custom|OnReceiveNoLockOns
  (Effects|SetBeamNiagaraSystem (Variables|Default|GetCachedWeapon) "/Game/Effects/Fort_Effects/Gameplay/Weapons/NS_Bullet_Tracer_Fast_BR.NS_Bullet_Tracer_Fast_BR")
  (Ability|GameplayEffect|RemoveLooseGameplayTags (Variables|Default|GetOwningPawn) (Variables|Default|GetGameplay%20Tags))
  (CallFunction|ClearAutoAimTarget))

(event Custom|OnAbilityEnd
  ;[1 1 1] On Ability End
  (if (not (> (Variables|Firing|GetCartridgesLeftToFire) 0))
    (Variables|Firing|SetCartridgesLeftToFire (Math|Integer|Min(Integer) (Stats|GetBurstCountStat) (FortWeapon|GetRemainingAmmo (Variables|Default|GetCachedWeapon)))))
  (Variables|Firing|SetLastFinishFireTime (Utilities|Time|GetTimeSeconds))
  (bind _result (Math|Integer|DecrementInt (Variables|Firing|GetCartridgesLeftToFire))
    (:%20%20
      ;;[1 1 1] Burst Completed
      (Ability|GameplayEffect|RemoveLooseGameplayTags (Variables|Default|GetOwningPawn) (Variables|Default|GetGameplay%20Tags))
      (if (not (> _result 0))
        (Variables|AutoAimWeaponPawnComponent|SetCurrentLockOnCount (Math|Integer|Max(Integer) (- (Variables|AutoAimWeaponPawnComponent|GetCurrentLockOnCount) 1)))
        (Utilities|FlowControl|Sequence
          (:then_0
            (if (== (Variables|AutoAimWeaponPawnComponent|GetCurrentLockOnCount) 0)
              (CallFunction|OnReceiveNoLockOns)))
          (:then_1
            (bind _returnvalue (Utilities|Time|SetTimerbyEvent (EventDispatchers|CreateEvent :!event Custom|TryFireBurst) (Stats|GetFiringRate)))
            (Variables|Firing|SetRefireTimerHandle _returnvalue)))))))

(event Custom|TryFireBurst
  (if (or (Variables|Firing|GetbTriggerHeld) (and (Variables|Firing|GetbWantsAutoLockedOnFire) (> (Variables|AutoAimWeaponPawnComponent|GetCurrentLockOnCount) 0)))
    (Firing|FireBurst)))

(event Custom|TellServerTargeting
  (params (IsTargeting :type Boolean) (InWeapon :type Actor%20Object%20Reference))
  ;[1 1 1] Tell Server
  ;;[1 1 1] Multicast Targeting
  (CallFunction|MulticastTargeting IsTargeting InWeapon))

(event Custom|MulticastTargeting
  (params (IsTargeting :type Boolean) (InWeapon :type Actor%20Object%20Reference))
  (bind _returnvalue (Components|GetOwner))
  (bind _returnvalue_2 (Ability|GameplayCue|MakeGameplayCueParameters :NormalizedMagnitude 0.0 :RawMagnitude 1.0 :Location "0,0,0" :Normal "0,0,0" :Instigator _returnvalue :GameplayEffectLevel 1 :AbilityLevel 1 :bReplicateLocationWhenUsingMinimalRepProxy false))
  (bind _weapontargetingtag (Variables|Default|GetWeaponTargetingTag))
  ;[1 1 1] Tell Server
  ;;[1 1 1] Multicast Targeting
  (if IsTargeting
    (Projectile|AddGameplayCue%28Looping%29Local InWeapon _weapontargetingtag _returnvalue_2 false)
    (bind _asb_auto_aim_rifle (Utilities|Casting|CastToB_AutoAim_Rifle InWeapon))
    (Variables|Default|SetCachedWeapon _asb_auto_aim_rifle)
    (else
      (Projectile|RemoveGameplayCue%28Looping%29Local InWeapon _weapontargetingtag _returnvalue_2))))

(event Custom|TellServerLockOn
  (params (NewLockTarget :type Fort%20Pawn%20Object%20Reference))
  (bind _cachedweapon (Variables|Default|GetCachedWeapon))
  (bind _returnvalue (Actor|GetOwner _cachedweapon))
  (bind _returnvalue_2 (Fortnite|GetFortPlayerControllerfromActor _returnvalue))
  ;[1 1 1] Tell Server
  ;;[1 1 1] Locked On Target Changed
  (bind _output_get (Variables|Victim|SetRepLockTarget NewLockTarget))
  (Networking|ForceNetUpdate _output_get)
  (bind _replocktarget (Variables|Victim|GetRepLockTarget)
    (:then
      (Default|CallOnGainedLockTarget)
      ;; Display indicator above target's head - locally
      (bind _asfort_player_pawn (Utilities|Casting|CastToFortPlayerPawn _replocktarget))
      (bind _output_get_2 (Variables|Default|SetTargetIndicatedPlayer _asfort_player_pawn))
      (IndicatedManagement|AddActorsToIndicatedListWithFilter _returnvalue_2 (Utilities|Array|MakeArray _output_get_2) (Utilities|Array|MakeArray (Variables|Default|GetIndicatedActorFilter)) false false true))
    (:else
      (Default|CallOnLostLockTarget)
      (bind _asfort_player_pawn_2 (Utilities|Casting|CastToFortPlayerPawn (Variables|Default|GetTargetIndicatedPlayer)))
      (IndicatedManagement|AddActorsToIndicatedListWithFilter _returnvalue_2 (Utilities|Array|MakeArray _asfort_player_pawn_2) (Utilities|Array|MakeArray (Variables|Default|GetIndicatedActorFilter_Remove)) false true) ; Refreshing indicator with .01 duration so it gets removed
      (IndicatedManagement|RemoveActorFromIndicatedList _returnvalue_2 (Variables|Default|GetTargetIndicatedPlayer))
      (IndicatedManagement|RemoveActorFromStenciledList _returnvalue_2 (Variables|Default|GetTargetIndicatedPlayer))
      (Variables|Default|SetTargetIndicatedPlayer 0))))

(event EventTick
  (params (DeltaSeconds :type Float%20%28single-precision%29))
  ;[1 1 1] Event Tick
  (if (not (CallFunction|IsServerBotControlled))
    (CallFunction|TickVictimWarning)
    (CallFunction|TickReticle DeltaSeconds)
    (CallFunction|TickLockOns DeltaSeconds)))

(event EventEndPlay
  (params (EndPlayReason :type EEndPlayReason%20Enum :default "Destroyed"))
  ;[1 1 1] End Play
  ;;[0.1046 0.009721 0.1779 0.4] UMG Locked Target Visual
  (Class|FortWeaponComponentTargetState|ClearTrackedTarget (Variables|/SmartRifle/Gameplay/B_AutoAim_Rifle.B_AutoAim_Rifle_C|GetFortWeaponComponent_TargetState (Variables|Default|GetCachedWeapon)))
  (CallFunction|TellServerLockOn 0) ; Clear Locked Target and remove Locked Indicator
  (Variables|Firing|SetCartridgesLeftToFire 0)
  (Projectile|RemoveGameplayCue%28Looping%29Local (Variables|Default|GetCachedWeapon) (Variables|Default|GetWeaponTargetingTag) (Ability|GameplayCue|MakeGameplayCueParameters :NormalizedMagnitude 0.0 :RawMagnitude 0.0 :Location "0,0,0" :Normal "0,0,0" :Instigator (Components|GetOwner) :GameplayEffectLevel 1 :AbilityLevel 1 :bReplicateLocationWhenUsingMinimalRepProxy false))
  (bind _spawnedwarningsound (Variables|Default|GetSpawnedWarningSound)
    (:then
      (Audio|Components|Audio|Stop _spawnedwarningsound))
    (:else
      _))
  (Variables|Victim|SetWarningActive false)
  (CallFunction|SetEnabledLocalLockCameraModifier false))

(fn CanFireLockedOn
  (return (and (FortWeapon|CanFire (Variables|Default|GetCachedWeapon) true) (> (Variables|AutoAimWeaponPawnComponent|GetCurrentLockOnCount) 0))))

(fn CanFreeFire
  (return (FortWeapon|CanFire (Variables|Default|GetCachedWeapon) true)))

(fn CanLockOn
  (bind _cachedweapon (Variables|Default|GetCachedWeapon)
    (:then
      (if (Math|Boolean|ANDBoolean (not (FortWeapon|IsReloading _cachedweapon)) (not (FortWeapon|IsFiring _cachedweapon)) (Camera|GetIsTargeting _cachedweapon) (Utilities|IsValid (Variables|AutoAimWeaponPawnComponent|GetLockOnTarget) :!class /Script/Engine.KismetSystemLibrary) (> (- (Utilities|Time|GetTimeSeconds) (Variables|Firing|GetLastFinishFireTime)) (ScalableFloat|GetValueatLevel (Variables|Lockon|Config|GetAfterFiringLockDelay))) (not (Utilities|Time|IsTimerActivebyHandle (Variables|Firing|GetRefireTimerHandle))))
        (bind _isallowedtolock (CallFunction|IsAllowedtoLock (Variables|AutoAimWeaponPawnComponent|GetLockOnTarget)))
        (return _isallowedtolock)
        (else
          (return false))))
    (:else
      (return false))))

(fn CanTickLocalShooter
  (Variables|Default|GetCachedWeapon
    (:then
      (return (FortWeapon|IsEquipped (Variables|Default|GetCachedWeapon))))
    (:else
      (return false))))

(fn ClearAutoAimTarget
  (Utilities|IsValid (Variables|AutoAimWeaponPawnComponent|GetLockOnTarget)
    (:Is%20Valid
      (Variables|AutoAimWeaponPawnComponent|SetLockOnTarget 0)
      (Default|CallOnLostLockTarget)
      ;[0.1046 0.009721 0.1779 0.4] UMG Locked Target Visual
      (Class|FortWeaponComponentTargetState|ClearTrackedTarget (Variables|/SmartRifle/Gameplay/B_AutoAim_Rifle.B_AutoAim_Rifle_C|GetFortWeaponComponent_TargetState (Variables|Default|GetCachedWeapon)))
      (if (> (Variables|AutoAimWeaponPawnComponent|GetCurrentLockOnCount) 0)
        (Audio|PlaySound2D "/Smartgun/Gameplay/StackingLockons/Sounds/AutoAim_Target_Lost_Cue.AutoAim_Target_Lost_Cue"))
      (CallFunction|ResetLockonCount)
      (Variables|AutoAimWeaponPawnComponent|SetProgressTowardNextLockOn 0.0)
      (CallFunction|TellServerLockOn 0))))

(fn FireAutomaticBurst
  (bind _outcanfire (CallFunction|CanFreeFire))
  (if _outcanfire
    (Firing|TrytoFireAutoBurst)
    (else
      (Variables|Firing|SetbWantsAutoLockedOnFire false))))

(fn FireBurst
  (if (== (Variables|Firing|GetCartridgesLeftToFire) 0)
    (bind _outcanfire (CallFunction|CanFireLockedOn))
    (if _outcanfire
      (Firing|TrytoFireLockedOn)
      ;[0.1046 0.009721 0.1779 0.4] UMG Locked Target Visual
      (Class|FortWeaponComponentTargetState|ClearTrackedTarget (Variables|/SmartRifle/Gameplay/B_AutoAim_Rifle.B_AutoAim_Rifle_C|GetFortWeaponComponent_TargetState (Variables|Default|GetCachedWeapon)))
      (else
        (Variables|Firing|SetbWantsAutoLockedOnFire false)
        (bind _outcanfire_2 (CallFunction|CanFreeFire))
        (if _outcanfire_2
          (Firing|TrytoFireAutoBurst))))))

(fn GetAngleToTarget)

(fn GetBulletTargetLocation
  (bind _lockontarget (Variables|AutoAimWeaponPawnComponent|GetLockOnTarget))
  (bind _lockontarget_2 (Variables|AutoAimWeaponPawnComponent|GetLockOnTarget))
  (bind _capsulecomponent (Variables|/Script/Engine.Character|GetCapsuleComponent _lockontarget_2))
  (bind _mesh (Variables|/Script/Engine.Character|GetMesh _lockontarget))
  (Utilities|FlowControl|Sequence
    (:then_0
      (Utilities|IsValid _lockontarget
        (:Is%20Valid
          (Utilities|FlowControl|Sequence
            (:then_0
              (Utilities|Casting|CastToNPC_Pawn_Wildlife_Parent _lockontarget
                (:then
                  (Utilities|IsValid _capsulecomponent
                    (:Is%20Valid
                      (return (Physics|GetCenterOfMass _capsulecomponent "None")))))
                (:CastFailed
                  (Utilities|IsValid _mesh
                    (:Is%20Valid
                      (switch int (Variables|AutoAimWeaponPawnComponent|GetCurrentLockOnCount)
                        (:1
                          (if (== (Stats|GetBurstCountStat) 10)
                            (return (Transformation|GetSocketLocation _mesh "FX_Head"))
                            (else
                              (return (Transformation|GetSocketLocation _mesh "FX_Chest")))))
                        (:2
                          (return (Transformation|GetSocketLocation _mesh "SM_R_Shoulders")))
                        (:3
                          (return (Transformation|GetSocketLocation _mesh "FX_Chest")))
                        (:4
                          (return (Transformation|GetSocketLocation _mesh "SM_Pelvis")))
                        (:5
                          (return (Transformation|GetSocketLocation _mesh "SM_Pelvis")))
                        (:6
                          (return (Transformation|GetSocketLocation _mesh "SM_Pelvis")))
                        (:7
                          (return (Transformation|GetSocketLocation _mesh "SM_Pelvis")))
                        (:8
                          (return (Transformation|GetSocketLocation _mesh "SM_Pelvis")))
                        (:9
                          (return (Transformation|GetSocketLocation _mesh "SM_Pelvis")))
                        (:10
                          (return (Transformation|GetSocketLocation _mesh "SM_Pelvis")))))))))
            (:then_1
              ;[1 0.4635 0.4635] We have a valid target, but did not find a good output location. Return the target actor's location.
              (return (Transformation|GetActorLocation _lockontarget)))))))
    (:then_1
      ;[1 0.4635 0.4635] Complete failure, return zero vector.
      (return "0, 0, 0"))))

(fn GetBurstCountStat
  (bind (_bfound _outrow) (Stats|GetWeaponStats))
  (bind (_jumpingfallingspreaddownsights _buseconstantcooldownrecoveryspeed _constantcooldownrecoveryspeed _aiathenahearlightweightprojectileimpactnoiserange _aiathenahearlightweightprojectileflybynoiserange _aiathenalightweightprojectileflybyemissioninterval _buseconstantlerpforrecoil _bstartrecoildelayafterrecoilpeaks _buseaccumulatedrecoiltolerpmintomaxperbulletrecoilvalues _recoilmintomaxcurve _recoilminverticalcurve _recoilmaxverticalcurve _recoilminhorizontalcurve _recoilmaxhorizontalcurve _recoildelay _bdelayspreadwithrecoil _bdisableoneframefiringanimationdelay _busescopedperbulletrecoiloverridesforbasicads _blerpsprintingspreadmultiplierwithmovementspeed _spreadmultipliercap _cartridgeperfire) (Utilities|Struct|BreakFortRangedWeaponStats _outrow))
  (if _bfound
    (return _cartridgeperfire)
    (else
      (return 0))))

(fn GetFiringRate
  (bind (_bfound _outrow) (Stats|GetWeaponStats))
  (bind (_jumpingfallingspreaddownsights _buseconstantcooldownrecoveryspeed _constantcooldownrecoveryspeed _firingrate) (Utilities|Struct|BreakFortRangedWeaponStats _outrow))
  (if _bfound
    (return (Math|Float|SafeDivide 1.0 _firingrate))
    (else
      (return 1.0))))

(fn GetLockOnTime
  (params (LockOnCount :type Integer))
  (return (ScalableFloat|GetValueatLevel (Variables|LockOn|Config|GetLockOnTime) (Math|Conversions|ToFloat(Integer) LockOnCount))))

(fn GetMaxLockOnsBasedOnAmmo
  (bind _output_get (LocalVariables|SetAmmoAllowedLocks (Math|Float|Ceil (Math|Float|SafeDivide (Math|Conversions|ToFloat(Integer) (FortWeapon|GetLocalMagazineAmmoCount (Variables|Default|GetCachedWeapon))) (Math|Conversions|ToFloat(Integer) (Stats|GetBurstCountStat))))))
  (return (Math|Integer|Min(Integer) _output_get (ScalableFloat|GetValueasInteger (Variables|AutoAimWeaponPawnComponent|GetMaxLockOns)))))

(fn GetWeaponStats
  (bind _cachedweapon (Variables|Default|GetCachedWeapon)
    (:then
      (bind (_outrow _returnvalue) (Fortnite|GetRangedWeaponStatsRow (Data|GetWeaponDataWeaponStatHandle _cachedweapon)))
      (return _returnvalue _outrow))
    (:else
      (return false 0))))

(fn InitializeAutoAimComponent
  (bind _asfort_player_pawn (Utilities|Casting|CastToFortPlayerPawn (Components|GetOwner))
    (:then
      (bind _output_get (Variables|Default|SetOwningPawn _asfort_player_pawn))
      (Variables|Firing|SetbTriggerHeld (Variables|/Script/FortniteGame.FortPawn|GetbPrimaryInputHeld _output_get))
      (Variables|Firing|SetCartridgesLeftToFire 0)
      (Utilities|FlowControl|Sequence
        (:then_0
          (Components|Tick|SetComponentTickEnabled (or (CallFunction|IsServerBotControlled) (and (not (Networking|IsServer)) (Pawn|IsLocallyControlled (Variables|Default|GetOwningPawn))))))
        (:then_1
          (bind _asb_auto_aim_rifle (Utilities|Casting|CastToB_AutoAim_Rifle (Weapon|GetCurrentWeapon (Variables|Default|GetOwningPawn))))
          (Variables|Default|SetCachedWeapon _asb_auto_aim_rifle))))
    (:CastFailed
      (Development|PrintString "SmartRifle: No Owning Pawn" true false "(R=0.000000,G=0.660000,B=1.000000,A=1.000000)" 2.0 "SmartRiflePawn"))))

(fn IsAllowedToLock
  (params (LockOnTarget :type Fort%20Pawn%20Object%20Reference))
  (if (FortPawn|IsDead LockOnTarget)
    (return false)
    (else
      (if (GameplayTags|HasAnyMatchingGameplayTags LockOnTarget (Variables|Lockon|Config|GetOverride%20Lock%20On%20Tag%20Container))
        (return true)
        (else
          (return (not (GameplayTags|HasAnyMatchingGameplayTags LockOnTarget (Variables|Lockon|Config|GetPrevent%20Lock%20On%20Tag%20Container)))))))))

(fn IsServerBotControlled
  (Networking|SwitchHasAuthority
    (:Authority
      (return (Pawn|IsBotControlled (Variables|Default|GetOwningPawn))))
    (:Remote
      (return false))))

(fn OnConsumeAmmo
  (params (Quantity :type Integer))
  (Variables|Firing|SetLastAmmoConsumeTime (Utilities|Time|GetTimeSeconds)))

(fn OnEquipped
  (CallFunction|InitializeAutoAimComponent))

(fn OnReload
  (CallFunction|ResetLockonCount)
  (Variables|AutoAimWeaponPawnComponent|SetProgressTowardNextLockOn 0.0)
  (Variables|Firing|SetbWantsAutoFreeFire false)
  (Variables|Firing|SetbWantsAutoLockedOnFire false)
  (Variables|Firing|SetCartridgesLeftToFire 0)
  (Utilities|Time|ClearandInvalidateTimerbyHandle (Variables|Firing|GetRefireTimerHandle)))

(fn OnRep_RepLockTarget
  (bind _replocktarget (Variables|Victim|GetRepLockTarget))
  (if (not (Networking|IsServer))
    (bind _returnvalue (Development|GetConsoleVariableBoolValue "SmartGun.EnableBeingLockedOnWidget"))
    (if _returnvalue
      (if (Variables|Victim|GetWarningActive)
        (Utilities|IsValid _replocktarget
          (:Is%20Valid
            (if (!= (Variables|Victim|GetLocalLockTarget) _replocktarget)
              _))
          (:Is%20Not%20Valid
            _))
        (Variables|Victim|SetWarningActive false)
        (CallFunction|SetEnabledLocalLockCameraModifier false)
        (bind _spawnedwarningsound (Variables|Default|GetSpawnedWarningSound)
          (:then
            (if (Audio|Components|Audio|IsPlaying _spawnedwarningsound)
              (Audio|Components|Audio|Stop _spawnedwarningsound)
              (CallFunction|SetEnabledLocalLockCameraModifier false)))
          (:else))
        (else
          (Utilities|IsValid _replocktarget
            (:Is%20Valid
              (if (Pawn|IsLocallyViewed _replocktarget)
                (Variables|Victim|SetLocalLockTarget _replocktarget)
                (Variables|Victim|SetWarningActive true)
                (CallFunction|SetEnabledLocalLockCameraModifier true)
                (bind _spawnedwarningsound_2 (Variables|Default|GetSpawnedWarningSound)
                  (:then
                    (if (Audio|Components|Audio|IsPlaying _spawnedwarningsound_2)
                      (Audio|Components|Audio|Stop _spawnedwarningsound_2)
                      (else
                        _))
                    (bind _returnvalue_2 (Audio|SpawnSoundAttached "/Smartgun/Gameplay/StackingLockons/Sounds/AutoAim_Target_Lock_Enemy_Cue.AutoAim_Target_Lock_Enemy_Cue" (Variables|/Script/Engine.Actor|GetRootComponent (Variables|Victim|GetRepLockTarget))))
                    (Variables|Default|SetSpawnedWarningSound _returnvalue_2))
                  (:else
                    _))))))))))

(fn OnRep_TimeTillLockOn)

(fn OnTriggerPress
  (bind _cachedweapon (Variables|Default|GetCachedWeapon))
  (if (or (not (ScalableFloat|GetValueasBool (Variables|Default|GetAllowReportOutOfAmmo))) (not (FortWeapon|Ammo|IsOutOfAmmo _cachedweapon)))
    (Variables|Firing|SetbTriggerHeld true)
    (Firing|FireBurst)
    (else
      (FortWeapon|Ammo|TryReportOutOfAmmo _cachedweapon))))

(fn OnTriggerReleased
  (Variables|Firing|SetbTriggerHeld false)
  (Variables|Firing|SetbWantsAutoFreeFire false))

(fn OnUnequip
  (CallFunction|ClearAutoAimTarget)
  (Variables|Firing|SetbWantsAutoFreeFire false)
  (Variables|Firing|SetbWantsAutoLockedOnFire false)
  (Variables|Firing|SetCartridgesLeftToFire 0)
  (Utilities|Time|ClearandInvalidateTimerbyHandle (Variables|Firing|GetRefireTimerHandle)))

(fn ResetLockOnCount
  (Variables|AutoAimWeaponPawnComponent|SetCurrentLockOnCount 0)
  (CallFunction|OnReceiveNoLockOns))

(fn SetEnabledLocalLockCameraModifier
  (params (bSetActive :type Boolean))
  (Variables|Victim|GetLocalLockTargetCameraModifier
    (:then
      (Variables|/Smartgun/Gameplay/StackingLockons/ScreenFX/CameraModifier_AutoAimVictim.CameraModifier_AutoAimVictim_C|SetIsActive (Variables|GetbSetActive) (Variables|Victim|GetLocalLockTargetCameraModifier)))
    (:else
      (bind _outmodifier (Class|WeaponFunctionLibrary|FindorAddCameraModifierforPawn "/Smartgun/Gameplay/StackingLockons/ScreenFX/CameraModifier_AutoAimVictim.CameraModifier_AutoAimVictim_C" (Variables|Victim|GetLocalLockTarget)))
      (bind _ascamera_modifier_auto_aim_victim (Utilities|Casting|CastToCameraModifier_AutoAimVictim _outmodifier))
      (bind _output_get (Variables|Victim|SetLocalLockTargetCameraModifier _ascamera_modifier_auto_aim_victim))
      (Variables|/Smartgun/Gameplay/StackingLockons/ScreenFX/CameraModifier_AutoAimVictim.CameraModifier_AutoAimVictim_C|SetIsActive (Variables|GetbSetActive) _output_get))))

(fn SetMaterialLockonParams
  (params (Count :type Integer))
  (bind _cachedweapon (Variables|Default|GetCachedWeapon)
    (:then
      (bind _returnvalue (Fortnite|GetDynamicMaterialInstance (Variables|/Script/FortniteGame.FortWeapon|GetWeaponMesh _cachedweapon)))
      (Rendering|Material|SetScalarParameterValue _returnvalue "Lockon_Progress" (Math|Conversions|ToFloat(Integer) Count) :!class /Script/Engine.MaterialInstanceDynamic)
      (if (< Count (ScalableFloat|GetValueatLevel (Variables|AutoAimWeaponPawnComponent|GetMaxLockOns)))
        (Variables|Default|SetMID_MaxLock false)
        (Rendering|Material|SetScalarParameterValue _returnvalue "Lockon_Blink" 0.0 :!class /Script/Engine.MaterialInstanceDynamic)
        (bind _blinkingaudioongun (Variables|Default|GetBlinkingAudioOnGun)
          (:then
            (Audio|Components|Audio|FadeOut _blinkingaudioongun 0.05 0.0 "Linear" :!class /Script/Engine.AudioComponent))
          (:else))
        (else
          (Variables|Default|SetMID_MaxLock true)
          (Rendering|Material|SetScalarParameterValue _returnvalue "Lockon_Blink" 1.0 :!class /Script/Engine.MaterialInstanceDynamic)
          ;[0.6 0.4 0.8] Play Blinking Audio when locked on
          (Variables|Default|GetBlinkingAudioOnGun
            (:then)
            (:else
              (bind _returnvalue_2 (Audio|SpawnSoundAttached "/SmartRifle/Audio/MSS_AutoAimRifle_WeaponLockedOn_LightsBlink.MSS_AutoAimRifle_WeaponLockedOn_LightsBlink" (Variables|/Script/Engine.Actor|GetRootComponent (Components|GetOwner))))
              (bind _output_get (Variables|Default|SetBlinkingAudioOnGun _returnvalue_2))
              ;[0.6 0.4 0.8] Optional Float set if gameplay decides to make a variable for it
              (Audio|Parameter|SetFloatParameter _output_get "BlinkSpeed" 2.0 :!class /Script/AudioExtensions.AudioParameterControllerInterface))))))
    (:else)))

(fn TargetingChanged
  (params (NewTargeting :type Boolean))
  (CallFunction|TellServerTargeting NewTargeting (Variables|Default|GetCachedWeapon)))

(fn TickDebug
  (Variables|Default|GetbEnableDebug
    (:then
      (Development|PrintString (Utilities|String|ToString(Text) (Utilities|Text|FormatText "NSLOCTEXT(\"[7604D4433BE9534751DF70D3E53B1E9C]\", \"2393B75A47929742BC7EB68E89C15D1C\", \"Smart Rifle\\r\\nLock Ons: {LockOns}\\r\\nTime TIll Next Lock On: {TimeTillLockOn}\\r\\nTime To Lock On: {TimeToLockOn}\\r\\nCartridges Left To Fire: {CartidgesLeft}\")" (CallFunction|GetCurrentLockonCount) (Variables|Lockon|GetTimeTillLockOn) (CallFunction|GetLockonTime (CallFunction|GetCurrentLockonCount)) (Variables|Firing|GetCartridgesLeftToFire))) true false "(R=0.000000,G=0.660000,B=1.000000,A=1.000000)" 1.0 "SmartRifle"))
    (:else)))

(fn TickLockOns
  (params (Delta%20Seconds :type Float%20%28double-precision%29))
  (if (Camera|GetIsTargeting (Variables|Default|GetCachedWeapon))
    (if (< (Variables|AutoAimWeaponPawnComponent|GetCurrentLockOnCount) (FortWeapon|GetLocalMagazineAmmoCount (Variables|Default|GetCachedWeapon)))
      (Utilities|FlowControl|Sequence
        (:then_0
          (if (Variables|Lockon|Config|GetbUseReticleDistanceNormalization)
            (Variables|AutoAimWeaponPawnComponent|SetTargetToReticleDistanceNormalized (Utilities|RemapFloat (Math|Float|Clamp(Float) (Math|Float|SafeDivide (Variables|Reticle|GetReticleDistance) (Variables|Reticle|GetInnerReticleSize))) "(X=0.100000,Y=1.000000)" "(X=0.000000,Y=1.000000)" true))))
        (:then_1
          (Variables|AutoAimWeaponPawnComponent|GetLockOnTarget
            (:then
              (if (< (Variables|AutoAimWeaponPawnComponent|GetCurrentLockOnCount) (CallFunction|GetMaxLockOnsBasedonAmmo))
                (bind _outcanlock (CallFunction|CanLockOn))
                (bind _output_get (Variables|Lockon|SetWasLockAllowed _outcanlock))
                (Default|CallOnLockonAllowed _output_get)
                (if _output_get
                  (LocalVariables|SetTimeToLockOn (CallFunction|GetLockonTime (CallFunction|GetCurrentLockonCount)))
                  (Variables|AutoAimWeaponPawnComponent|SetProgressTowardNextLockOn (+ (Variables|AutoAimWeaponPawnComponent|GetProgressTowardNextLockOn) (Variables|GetDelta%20Seconds)))
                  (Utilities|FlowControl|Sequence
                    (:then_0
                      (if (Variables|Lockon|Config|GetbUseReticleDistanceNormalization)
                        ;[1 1 1] Previous implementation however its a bit odd as it doesn't progress things faster or slower, its just a TRUE/FALSE
                        (Variables|Lockon|SetTimeTillLockOn (Math|Float|Lerp (ScalableFloat|GetValueatLevel (Variables|AutoAimWeaponPawnComponent|GetLockOnTimeReticleCenter)) (ScalableFloat|GetValueatLevel (Variables|AutoAimWeaponPawnComponent|GetLockOnTimeReticleEdge)) (Variables|AutoAimWeaponPawnComponent|GetTargetToReticleDistanceNormalized))) ; For UI to display charging progress
                        (else
                          (Variables|Lockon|SetTimeTillLockOn (- (LocalVariables|GetTimeToLockOn) (Variables|AutoAimWeaponPawnComponent|GetProgressTowardNextLockOn))) ; For UI to display charging progress
                        )))
                    (:then_1
                      ;[1 1 1] Have we completed our lock on?
                      (if (>= (Variables|AutoAimWeaponPawnComponent|GetProgressTowardNextLockOn) (LocalVariables|GetTimeToLockOn))
                        ;[1 1 1] Make sure to accommodate any time overflow
                        (Variables|AutoAimWeaponPawnComponent|SetProgressTowardNextLockOn (Math|Float|Absolute(Float) (- (LocalVariables|GetTimeToLockOn) (Variables|AutoAimWeaponPawnComponent|GetProgressTowardNextLockOn))))
                        (Utilities|FlowControl|Sequence
                          (:then_0
                            (bind _result (Math|Integer|IncrementInt (Variables|AutoAimWeaponPawnComponent|GetCurrentLockOnCount))
                              (:%20%20
                                (Game|Feedback|ClientPlayForceFeedback (Fortnite|GetFortPlayerControllerfromActor (Variables|Default|GetOwningPawn)) "/Game/Weapons/ForceFeedback/FF_LockOn_Tick.FF_LockOn_Tick" "None" false false false)
                                (if (== _result (CallFunction|GetMaxLockOnsBasedonAmmo))
                                  (CallFunction|OnReceiveFirstLockOn)))))
                          (:then_1
                            (if (== (Variables|AutoAimWeaponPawnComponent|GetCurrentLockOnCount) (CallFunction|GetMaxLockOnsBasedonAmmo))
                              (Ability|GameplayEffect|AddLooseGameplayTags (Variables|Default|GetOwningPawn) (Variables|Default|GetGameplay%20Tags))
                              (Audio|PlaySound2D "/SmartRifle/Audio/AutoAimRifle_Target_Lock_Lock_Cue.AutoAimRifle_Target_Lock_Lock_Cue") ; Final Lock
                              (Default|CallOnFullCharge)
                              (else
                                (bind _returnvalue (Audio|SpawnSound2D "/SmartRifle/Audio/AutoAimRifle_Target_Count_Cue.AutoAimRifle_Target_Count_Cue")) ; Lock Pip
                                (Audio|Components|Audio|SetPitchMultiplier _returnvalue (Math|Float|MapRangeClamped (Math|Conversions|ToFloat(Integer) (Variables|AutoAimWeaponPawnComponent|GetCurrentLockOnCount)) 0.0 8.0 1.0 2.0))
                                (Audio|PlaySound2D "/SmartRifle/Audio/AutoAimRifle_Target_Lock_Appears_Cue.AutoAimRifle_Target_Lock_Appears_Cue") ; Chevron
                              ))))))))))
            (:else
              (Variables|AutoAimWeaponPawnComponent|SetProgressTowardNextLockOn 0.0)
              (CallFunction|ResetLockonCount))))))
    (else
      (CallFunction|ResetLockonCount))))

(fn TickReticle
  (params (DeltaSeconds :type Float%20%28double-precision%29))
  (bind _lockontarget (Variables|AutoAimWeaponPawnComponent|GetLockOnTarget))
  (if (not (Networking|IsServer))
    (Variables|Default|GetCachedWeapon
      (:then
        (if (Camera|GetIsTargeting (Variables|Default|GetCachedWeapon))
          (Default|SetMaterialLockonParams (Variables|AutoAimWeaponPawnComponent|GetCurrentLockOnCount))
          (Variables|AutoAimWeaponPawnComponent|GetLockOnTarget
            (:then
              ;[0 0.1864 1] Clear auto aim target when no longer valid
              (bind _isallowedtolock (CallFunction|IsAllowedtoLock (Variables|AutoAimWeaponPawnComponent|GetLockOnTarget)))
              (if _isallowedtolock
                (bind (_outtargetpawn _outreticledistance) (AutoAimWeapon|FindBestAutoAimTarget (Variables|Default|GetOwningPawn) (Variables|Reticle|GetOuterReticleSize) (ScalableFloat|GetValueatLevel (Variables|LockOn|Config|GetLockOnRange))))
                (Variables|Reticle|SetReticleDistance _outreticledistance)
                (Utilities|IsValid _outtargetpawn
                  (:Is%20Valid
                    (if (!= _outtargetpawn (Variables|AutoAimWeaponPawnComponent|GetLockOnTarget))
                      ;[1 1 1] Target Switch - Only if the old target left the reticle
                      (bind (_outreticledistance_2 _returnvalue) (AutoAimWeapon|DoesAutoAimWeaponReticleIntersectTarget (Variables|Default|GetOwningPawn) (Variables|AutoAimWeaponPawnComponent|GetLockOnTarget) (Variables|Reticle|GetInnerReticleSize)))
                      (if _returnvalue
                        (bind _returnvalue_2 (AutoAimWeapon|DoesAutoAimWeaponHaveLineOfSight (Variables|Default|GetOwningPawn) (Variables|AutoAimWeaponPawnComponent|GetLockOnTarget)))
                        (if _returnvalue_2
                          _
                          (else
                            _)))
                      (CallFunction|ResetLockonCount)
                      (Variables|AutoAimWeaponPawnComponent|SetLockOnTarget _outtargetpawn)
                      (Variables|Reticle|SetTimeSinceLostTarget 0.0)
                      ;;[0.1046 0.009721 0.1779 0.4] UMG Locked Target Visual
                      (Class|FortWeaponComponentTargetState|TrackTarget (Variables|/SmartRifle/Gameplay/B_AutoAim_Rifle.B_AutoAim_Rifle_C|GetFortWeaponComponent_TargetState (Variables|Default|GetCachedWeapon)) _lockontarget)
                      (CallFunction|TellServerLockOn _lockontarget)
                      (Default|CallOnGainedLockTarget)
                      (else
                        (Variables|Reticle|SetTimeSinceLostTarget 0.0))))
                  (:Is%20Not%20Valid
                    ;[1 1 1] Lost Target
                    (bind _output_get (Variables|Reticle|SetTimeSinceLostTarget (+ (Variables|Reticle|GetTimeSinceLostTarget) (Variables|GetDeltaSeconds))))
                    (bind (_outreticledistance_3 _returnvalue_3) (AutoAimWeapon|DoesAutoAimWeaponReticleIntersectTarget (Variables|Default|GetOwningPawn) (Variables|AutoAimWeaponPawnComponent|GetLockOnTarget) (Variables|Reticle|GetOuterReticleSize)))
                    (if (not (and _returnvalue_3 (< _output_get (ScalableFloat|GetValueatLevel (Variables|Default|GetLockOnForgivenessTime)))))
                      (CallFunction|ClearAutoAimTarget))))
                (else
                  (CallFunction|ClearAutoAimTarget)
                  (Default|SetMaterialLockonParams 0))))
            (:else
              ;[1 1 1] Looking for target
              (bind _outtargetpawn_2 (AutoAimWeapon|FindBestAutoAimTarget (Variables|Default|GetOwningPawn) (Variables|Reticle|GetInnerReticleSize) (ScalableFloat|GetValueatLevel (Variables|LockOn|Config|GetLockOnRange))))
              (Utilities|IsValid _outtargetpawn_2
                (:Is%20Valid
                  (Variables|AutoAimWeaponPawnComponent|SetLockOnTarget _outtargetpawn_2)
                  (Variables|Reticle|SetTimeSinceLostTarget 0.0)
                  (CallFunction|TellServerLockOn (Variables|AutoAimWeaponPawnComponent|GetLockOnTarget))
                  (Default|CallOnGainedLockTarget)
                  ;;[0.1046 0.009721 0.1779 0.4] UMG Locked Target Visual
                  (Class|FortWeaponComponentTargetState|TrackTarget (Variables|/SmartRifle/Gameplay/B_AutoAim_Rifle.B_AutoAim_Rifle_C|GetFortWeaponComponent_TargetState (Variables|Default|GetCachedWeapon)) (Variables|AutoAimWeaponPawnComponent|GetLockOnTarget))))))
          (else
            _)))
      (:else
        _))
    (Variables|AutoAimWeaponPawnComponent|GetLockOnTarget
      (:then
        (CallFunction|ClearAutoAimTarget)
        (Default|SetMaterialLockonParams 0))
      (:else))))

(fn TickVictimWarning
  (if (and (Utilities|IsValid (Variables|Default|GetSpawnedWarningSound) :!class /Script/Engine.KismetSystemLibrary) (not (Utilities|IsValid (Variables|Victim|GetRepLockTarget) :!class /Script/Engine.KismetSystemLibrary)))
    (Audio|Components|Audio|Stop (Variables|Default|GetSpawnedWarningSound)))
  (Variables|Victim|GetRepLockTarget
    (:then
      (if (or (FortPawn|IsDead (Variables|Victim|GetRepLockTarget)) (Health|IsDBNO (Variables|Victim|GetRepLockTarget)))
        (Variables|Victim|SetWarningActive false)
        (CallFunction|SetEnabledLocalLockCameraModifier false)))
    (:else
      (Variables|Victim|SetWarningActive false))))

(fn TryToFireAutoBurst
  (Variables|Firing|SetbWantsAutoFreeFire (Variables|Firing|GetbTriggerHeld))
  (if (FortWeapon|CanFire (Variables|Default|GetCachedWeapon) true)
    (Class|FortWeapon|TrytoCustomFire (Variables|Default|GetCachedWeapon) false)))

(fn TryToFireLockedOn
  (bind _lockontarget (Variables|AutoAimWeaponPawnComponent|GetLockOnTarget))
  (Variables|Firing|SetbWantsAutoLockedOnFire false)
  (if (> (Variables|AutoAimWeaponPawnComponent|GetCurrentLockOnCount) 0)
    (Utilities|IsValid _lockontarget
      (:Is%20Valid
        (if (FortPawn|IsDead _lockontarget)
          (Variables|Firing|SetbWantsAutoLockedOnFire false)
          (else
            (Class|FortWeapon|TrytoCustomFire (Variables|Default|GetCachedWeapon) false)
            (Variables|Firing|SetbWantsAutoLockedOnFire true))))
      (:Is%20Not%20Valid
        (Variables|Firing|SetbWantsAutoLockedOnFire false)))
    (else
      (Variables|Firing|SetbWantsAutoLockedOnFire false))))
