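; Smart Rifle weapon Blueprint: an auto-aim rifle that steers shots toward locked-on targets.
; It finds and caches AutoAim_Rifle_Component on the instigator, forwarding trigger, targeting, ammo, reload and equip events to it.
; An aim-rotation override turns camera aim toward the lock target, or optionally adds spread without a lock.
; Lock state also drives the downsights spread override, tracer and muzzle Niagara parameters, and an extra fire sound.
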(event EventOnWrapApplied
  (params (WrapDefinition :type Athena%20Item%20Wrap%20Definition%20Object%20Reference))
  ;[1 1 1] re updating dynamic material instance after wrap is applyed.
  (bind _returnvalue (Fortnite|GetDynamicMaterialInstance (Variables|Visual|GetWeaponMesh) 1))
  (Variables|Default|SetMID _returnvalue))

(event EventBeginPlay
  (|CallParentFunction)
  (CallFunction|GetSmartComponentCached)
  (EventDispatchers|BindEventtoOnLocalTargetingChanged (EventDispatchers|CreateEvent :!event Custom|TargetingChanged))
  (EventDispatchers|BindEventtoOnPressTrigger (EventDispatchers|CreateEvent :!event Custom|TriggerPressed))
  (EventDispatchers|BindEventtoOnReleaseTrigger (EventDispatchers|CreateEvent :!event Custom|TriggerReleased))
  (EventDispatchers|BindEventtoOnConsumeAmmo (EventDispatchers|CreateEvent :!event Custom|OnConsumedAmmo))
  (EventDispatchers|BindEventtoOnLocalReloadStarted (EventDispatchers|CreateEvent :!event Custom|ReloadStarted))
  (Variables|FortWeapon|SetOnGetAimRotOverride (EventDispatchers|CreateEvent :!event OnGetAimRot))
  (Variables|FortWeapon|SetOnBeamFiredDelegate (EventDispatchers|CreateEvent :!event Custom|OnTracerBeamFired)))

(event EventEndPlay
  (params (EndPlayReason :type EEndPlayReason%20Enum :default "Destroyed"))
  (bind _smartcomp (Variables|Default|GetSmartComp)
    (:then
      (Default|UnbindEventfromOnGainedLockTarget _smartcomp (EventDispatchers|CreateEvent :!event Custom|HandleTargetAcquired))
      (Default|UnbindEventfromOnLostLockTarget _smartcomp (EventDispatchers|CreateEvent :!event Custom|HandleTargetLost)))
    (:else)))

(event Custom|TargetingChanged
  (params (bIsTargeting :type Boolean))
  (bind _component (CallFunction|GetSmartComponentCached))
  (Utilities|IsValid _component
    (:Is%20Valid
      (Class|AutoAimRifleComponent|TargetingChanged _component bIsTargeting))))

(event Custom|OnAbilityEnd
  ;[1 1 1] This is called when firing completes.
  ; Ex:
  ; Single fire weapon fires
  ; Burst fire fully completes
  (bind _component (CallFunction|GetSmartComponentCached))
  (Utilities|IsValid _component
    (:Is%20Valid
      (Class|AutoAimRifleComponent|OnAbilityEnd _component))))

(event Custom|TriggerPressed
  (params (Weapon :type Fort%20Weapon%20Object%20Reference))
  (bind _component (CallFunction|GetSmartComponentCached))
  (Utilities|IsValid _component
    (:Is%20Valid
      (Firing|OnTriggerPress _component))))

(event Custom|TriggerReleased
  (params (Weapon :type Fort%20Weapon%20Object%20Reference))
  (bind _component (CallFunction|GetSmartComponentCached))
  (Utilities|IsValid _component
    (:Is%20Valid
      (Firing|OnTriggerReleased _component))))

(event Custom|OnConsumedAmmo
  (params (Quantity :type Integer))
  (bind _component (CallFunction|GetSmartComponentCached))
  (Utilities|IsValid _component
    (:Is%20Valid
      (Class|AutoAimRifleComponent|OnConsumeAmmo _component Quantity))))

(event Custom|ReloadStarted
  (params (ReloadTime :type Float%20%28single-precision%29) (ReloadType :type EFortWeaponReloadType%20Enum :default "ReloadWholeClip"))
  (bind _component (CallFunction|GetSmartComponentCached))
  (Utilities|IsValid _component
    (:Is%20Valid
      (Class|AutoAimRifleComponent|OnReload _component))))

(event FortWeapon|EventOnLocalUnEquip
  (bind _component (CallFunction|GetSmartComponentCached))
  (Utilities|IsValid _component
    (:Is%20Valid
      (Class|AutoAimRifleComponent|OnUnequip _component))))

(event FortWeapon|EventOnEquippedAndAttached
  (|CallParentFunction)
  (bind _component (CallFunction|GetSmartComponentCached))
  (Utilities|IsValid _component
    (:Is%20Valid
      (Class|AutoAimRifleComponent|OnEquipped _component))))

(event Custom|OnTracerBeamFired
  (params (Weapon :type Fort%20Weapon%20Object%20Reference) (bUsePersistentBeam :type Boolean) (PSC :type FXSystem%20Component%20Object%20Reference) (BeamOrigin :type Vector%20%28by%20ref%29) (HitLocation :type Vector%20%28by%20ref%29))
  (bind _asniagara_particle_system_component (Utilities|Casting|CastToNiagaraComponent PSC))
  (Niagara|SetNiagaraVariable(Vector3) _asniagara_particle_system_component "BeamStart_0_0" BeamOrigin)
  (Niagara|SetNiagaraVariable(Vector3) _asniagara_particle_system_component "BeamEnd_0" HitLocation)
  (Niagara|SetNiagaraVariable(Vector3) _asniagara_particle_system_component "Direction" (Math|Vector|GetRightVector (Transformation|GetWorldRotation (Variables|Visual|GetWeaponMesh))))
  (Niagara|SetNiagaraVariable(Bool) _asniagara_particle_system_component "IsTargeting" (CallFunction|HasLockOns))
  ; Modify Muzzle Flash when Locked On
  (bind _asniagara_particle_system_component_2 (Utilities|Casting|CastToNiagaraComponent (Variables|Effects|Muzzle|GetSpawned%20Muzzle%20System)))
  (Niagara|SetNiagaraVariable(Bool) _asniagara_particle_system_component_2 "LockedOn" (CallFunction|HasLockOns)))

(event Custom|HandleTargetAcquired
  ;[0 0.1106 1] Use override downsights spread (if valid) when locked on to a target.
  (FortWeapon|SetUseOverrideSpreadDownsightsMultiplier true))

(event Custom|HandleTargetLost
  ;[0 0.1106 1] Use override downsights spread (if valid) when locked on to a target.
  (FortWeapon|SetUseOverrideSpreadDownsightsMultiplier false))

(event EventPlayWeaponFireCustomAudio
  (params (bPersistentFire :type Boolean) (bSecondaryFire :type Boolean))
  (|CallParentFunction bPersistentFire bSecondaryFire)
  (bind _smartcomp (Variables|Default|GetSmartComp)
    (:then
      (bind _outcanfire (Class|AutoAimRifleComponent|CanFireLockedOn _smartcomp))
      (if _outcanfire
        (Audio|PlaySound2D "/SmartRifle/Audio/AutoAimRifle_Lock_Shot_Layer.AutoAimRifle_Lock_Shot_Layer") ; Locked Shot Layer
      ))
    (:else)))

(event Niagara|EventReceiveParticleData
  (params (Data :type Array%20of%20Basic%20Particle%20Data%20Structures) (NiagaraSystem :type Niagara%20System%20Object%20Reference) (SimulationPositionOffset :type Vector%20%28by%20ref%29 :default "0, 0, 0")))

(fn GetSmartComponentCached
  (bind _instigator (Variables|Actor|GetInstigator))
  (bind _returnvalue (Actor|GetComponentbyClass _instigator "/SmartRifle/Gameplay/AutoAim_Rifle_Component.AutoAim_Rifle_Component_C"))
  (bind _smartcomp (Variables|Default|GetSmartComp))
  (bind _smartcomp_2 (Variables|Default|GetSmartComp)
    (:then
      (Utilities|FlowControl|Sequence
        (:then_0
          (if (!= (Variables|/SmartRifle/Gameplay/AutoAim_Rifle_Component.AutoAim_Rifle_Component_C|GetCachedWeapon (Variables|Default|GetSmartComp)) self)
            (Variables|/SmartRifle/Gameplay/AutoAim_Rifle_Component.AutoAim_Rifle_Component_C|SetCachedWeapon self (Variables|Default|GetSmartComp))))
        (:then_1
          (return _smartcomp_2))))
    (:else
      (Utilities|IsValid _instigator
        (:Is%20Valid
          (Utilities|IsValid _returnvalue
            (:Is%20Valid
              (bind _output_get (Variables|Default|SetSmartComp _returnvalue))
              (Variables|/SmartRifle/Gameplay/AutoAim_Rifle_Component.AutoAim_Rifle_Component_C|SetCachedWeapon self _output_get)
              (Default|BindEventtoOnGainedLockTarget _smartcomp (EventDispatchers|CreateEvent :!event HandleTargetAcquired))
              (Default|BindEventtoOnLostLockTarget _smartcomp (EventDispatchers|CreateEvent :!event HandleTargetLost))
              (return _smartcomp))
            (:Is%20Not%20Valid
              _)))
        (:Is%20Not%20Valid
          _))
      (return 0))))

(fn HasLockOns
  (Variables|Default|GetSmartComp
    (:then
      (return (> (Class|AutoAimWeaponPawnComponent|GetCurrentLockonCount (Variables|Default|GetSmartComp) :!class /Script/AutoAimWeaponRuntime.AutoAimWeaponPawnComponent) 0)))
    (:else
      (return false))))

(fn IsAimedAtAutoAimTarget
  (bind (_location _rotation) (Pawn|GetPlayerViewPoint (Fortnite|GetFortPlayerControllerfromActor (Variables|Actor|GetInstigator))))
  (Utilities|IsValid (Variables|/Script/AutoAimWeaponRuntime.AutoAimWeaponPawnComponent|GetLockOnTarget (Variables|Default|GetSmartComp))
    (:Is%20Valid
      (bind _outhit (Collision|LineTraceByChannel _location (+ _location (* (Math|Vector|GetForwardVector _rotation) 5e+03)) "TraceTypeQuery6" false (Utilities|Array|MakeArray (Variables|Actor|GetInstigator))))
      (return (and (Collision|BreakHitResult _outhit) (== (.HitActor (Collision|BreakHitResult _outhit)) (Variables|/Script/AutoAimWeaponRuntime.AutoAimWeaponPawnComponent|GetLockOnTarget (Variables|Default|GetSmartComp))))))))

(fn OnGetAimRot
  (params (Weapon :type Fort%20Weapon%20Object%20Reference) (TargetingSource :type EFortAbilityTargetingSource%20Enum :default "Camera"))
  (if (and (Utilities|Enum|Equal(Enum) TargetingSource "Camera") (Utilities|IsValid (Variables|Default|GetSmartComp) :!class /Script/Engine.KismetSystemLibrary))
    (if (CallFunction|HasLockOns)
      ;[0 0.5781 0.02235] Nativized
      ; Rotate aim toward target if locked on.
      (bind _outrotation (Class|AutoAimWeaponPawnComponent|RotateAimTowardsLockonTarget (Variables|Default|GetSmartComp)))
      (return _outrotation true)
      (elif (Variables|Spread|GetbUseBlueprintSpread)
        ;[1 0.3196 0] Nativized
        ; Apply additional spread if targeting but not locked on
        (bind _outrotation_2 (Class|AutoAimWeaponPawnComponent|ApplySpreadtoAimRotation (Variables|Default|GetSmartComp) (ScalableFloat|GetValueatLevel (Variables|Spread|GetADSSpread))))
        (return _outrotation_2 true)
        (else
          (return "0,0.000000,1.000000" false))))
    (else
      (return "0,0.000000,1.000000" false))))

(fn ShouldCustomTriggerPerformPendingFire
  (return true))

(fn ConstructionScript
  (|CallParentFunction)
  (bind _returnvalue (Fortnite|GetDynamicMaterialInstance (Variables|Visual|GetWeaponMesh) 1))
  (Variables|Default|SetMID _returnvalue))
