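# Translucent, unlit, two-sided sci-fi trail material for the Smart Rifle auto-aim VFX, usable on Niagara ribbons, sprites and mesh particles.
# A panning flame noise blends EdgeColor, Color and Color2, tinted by particle color and boosted to HDR for emissive.
# A panning plasma mask, offset and thresholded by dynamic parameters, drives dissolve.
# Particle alpha, depth fade and procedural UV edge falloffs shape the opacity mask.
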Material /SmartRifle/Gameplay/VFX/M_AutoAim_Scifi_Trail
graph {
 properties {
  bUsedWithNiagaraMeshParticles=true
  bUsedWithNiagaraRibbons=true
  bUsedWithParticleSprites=true
  blendMode=BLEND_Translucent
  pixelDepthOffsetMode=PDOM_Legacy
  shadingModel=MSM_Unlit
  twoSided=true
  usedShadingModels=Unlit
 }
 nodes [
  n1 = VectorParameter() {parameterName=UVScale, defaultValue=<rgba 0 0 0 1>}
  n2 = Multiply(A=AppendVector(A=@n1.R, B=@n1.R), B=TextureCoordinate() {uTiling=0.25, vTiling=2.5})
  n3 = DynamicParameter() {defaultValue=<rgba 0 1 1 1>, paramNames [Dissolve,Offset_X,Offset_Y,Scale]}
  Add(A=@n2, B=AppendVector(A=@n3.Offset_X, B=@n3.Offset_X))
  n4 = DynamicParameter() {defaultValue=<rgba 0 1 1 1>, paramNames [Dissolve,Offset_X,Offset_Y,Scale]}
  n5 = AppendVector(A=@n4.Offset_X, B=@n4.Offset_X)
  n6 = VectorParameter() {parameterName=DissolveScale, defaultValue=<rgba 1 1 0 1>}
  n7 = Multiply(A=AppendVector(A=@n6.R, B=@n6.R), B=TextureCoordinate() {uTiling=0.1, vTiling=3})
  n8 = TextureSampleParameter2D(UVs=Panner(Coordinate=Add(A=@n5, B=@n7)) {speedX=-0.1, speedY=0.1}) {
   group=DissolveMask
   parameterName=TextureDissolve
   samplerType=SAMPLERTYPE_LinearGrayscale
   sortPriority=0
   texture=/Game/Effects/Fort_Effects/Emotes/UltraLight/Texture/T_Plasma_C_Greyscale
  }
  n9 = OneMinus(None=SmoothStep(Min=OneMinus(None=@n8.G), Value=DynamicParameter() {defaultValue=<rgba 0 1 1 1>, paramNames [Dissolve,Offset_X,Offset_Y,Scale]}.Dissolve))
  n10 = ParticleColor()
  n11 = ComponentMask(None=TextureCoordinate()) {g=true}
  n12 = Multiply(A=OneMinus(None=@n11), B=@n11)
  n13 = ComponentMask(None=TextureCoordinate()) {r=true}
  n14 = Abs(None=Multiply(A=Add(A=@n13) {constB=-0.5}) {constB=2})
  n15 = Saturate(None=Multiply(A=OneMinus(None=@n14)) {constB=2.5})
  n16 = Multiply(
   A=DepthFade(Opacity=Multiply(A=@n9, B=@n10.A), FadeDistance=ScalarParameter() {defaultValue=30, parameterName=FadeDistance})
   B=Multiply(A=@n12, B=Multiply(A=@n15, B=@n15))
  )
  n17 = TextureSampleParameter2D(UVs=Panner(Coordinate=@n2) {speedX=-0.25}) {parameterName=TextureTrail, texture=/Game/Effects/FogVolume/noise_flamy_Tile}
  n18 = Multiply(A=@n17.R, B=@n17.R)
  n19 = Multiply(A=@n18, B=@n18)
  LinearInterpolate(A=Power(Base=@n16), B=Multiply(A=@n16, B=@n19))
  n20 = VectorParameter() {parameterName=EdgeColor, defaultValue=<rgba 0.5 0.0583 0.2905 1>}
  n21 = VectorParameter() {parameterName=Color, defaultValue=<rgba 1 0.5 0.5 0>}
  n22 = VectorParameter() {parameterName=Color2, defaultValue=<rgba 1 0 0 0>}
  n23 = Multiply(A=@n10.RGB, B=Add(A=LinearInterpolate(A=@n20.RGB, B=@n21.RGB, Alpha=@n17.G), B=Multiply(A=@n22.RGB, B=@n19)))
  n24 = Clamp(None=@n23)
  n25 = ScalarParameter() {defaultValue=20, parameterName=HDRIntensity}
  MaterialFunctionCall(Emissive=Multiply(
    A=MaterialFunctionCall(LDR Input=@n24, HDR Tint=@n23, HDR Intensity Multiplier=@n25) {materialFunction=&DeriveHDRfromLDR}.RemappedRange
    B=@n12
   )) {materialFunction=&MF_FX_EmissiveAdjustments}
 ]
}
